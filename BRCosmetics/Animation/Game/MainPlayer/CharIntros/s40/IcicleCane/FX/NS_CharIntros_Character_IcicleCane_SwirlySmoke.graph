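# Niagara system for the swirly smoke effect in the IcicleCane character intro, built from a single mesh-particle emitter.
# A burst spawns two short-lived bright white meshes with random height, yaw spread evenly by normalized exec index.
# They spin around Z, drift upward and sit offset along the camera axis.
# Over life, curves stretch them vertically and fade a very faint alpha in and out, while dynamic material parameters add per-particle variation.
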NiagaraSystem /BRCosmetics/Animation/Game/MainPlayer/CharIntros/s40/IcicleCane/FX/NS_CharIntros_Character_IcicleCane_SwirlySmoke
dependencies {
 usedDataInterfaces [&NiagaraDataInterfaceVectorCurve, &NiagaraDataInterfaceCurve]
 usedDynamicInputs []
 usedModules [&SystemState]
}
emitters {SwirlSmoke {
  emitterUpdateScript [
   &EmitterState()
   &SpawnBurst_Instantaneous(Spawn Count=2, Use Spawn Probability=false, UseLoopCountLimit=false)
  ]
  particleSpawnScript [
   &InitializeParticle(
    Color=<rgba 5 5 5 1>
    Lifetime=0.7
    Mesh Scale=&RandomRangeVector(Minimum=<0.75 0.75 0.6>, Maximum=<0.75 0.75 1.4>)
    Mesh Scale Mode=::NewEnumerator4
    UsePositionOffset=false
    Write Material Random=true
   )
   &AddRotationalVelocity(Rotation Rate=<0 0 -0.6>)
   &InitialMeshOrientation(
    Mesh Orientation Mode=::NewEnumerator7
    Rotation=&MakeVector(X=0, Y=0, Z=&ReturnNormalizedExecIndex(Normalized Index Includes One=false))
    Use Rotation Vector=true
   )
   &CameraOffset(Camera Offset Amount=15)
   &AddVelocity(Velocity=<0 0 65>)
  ]
  particleUpdateScript [
   &ParticleState()
   &ScaleMeshSize(Scale Factor=&VectorFromCurve(
     VectorCurve=DataInterface() {propertyValues {
       XCurve {defaultValue=1, keys [<at 1 1.2>]}
       YCurve {defaultValue=1, keys [<at 1 1.2>]}
       ZCurve {
        defaultValue=1
        keys [
         {arriveTangent=2.7503, interpMode=RCIM_Cubic, leaveTangent=2.7503, tangentMode=RCTM_User, time=0, value=0.3}
         {arriveTangent=0.9, interpMode=RCIM_Cubic, leaveTangent=0.9, time=0.2, value=0.9}
         <at 1 1.2>
        ]
       }
      }}
     Scale Curve=<1 1 0.735>
    ))
   &ScaleColor(
    Scale Alpha=&FloatFromCurve(
     FloatCurve=DataInterface() {propertyValues {Curve {keys [
         {arriveTangent=7.9556, interpMode=RCIM_Cubic, time=0, value=0}
         {arriveTangentWeight=0.0667, interpMode=RCIM_Cubic, leaveTangentWeight=0.2667, tangentMode=RCTM_User, time=0.2, value=1}
         <at 1 0 RCIM_Cubic>
        ]}}}
     Scale Curve=0.05
    )
    ScaleA=true
    ScaleRGB=true
   )
   &DynamicMaterialParameters(
    Index 0 Param 1=1
    Index 0 Param 2=&RandomRangeFloat(Minimum=0, Maximum=2)
    Module.Index0Float0Write=true
    Module.Index0Float1Write=true
    Module.Index0Float2Write=false
    Module.Index0Float3Write=false
   )
   &SolveForcesAndVelocity(Clamp Velocity=false, Limit Acceleration=false)
  ]
  rendererClasses [&NiagaraMeshRendererProperties]
  renderers [{rendererClass=&NiagaraMeshRendererProperties, rendererIndex=0}]
 }}
summary {
 systemName=NS_CharIntros_Character_IcicleCane_SwirlySmoke
 emitters [{emitterName=SwirlSmoke, rendererClasses [&NiagaraMeshRendererProperties]}]
 userVariables []
}
