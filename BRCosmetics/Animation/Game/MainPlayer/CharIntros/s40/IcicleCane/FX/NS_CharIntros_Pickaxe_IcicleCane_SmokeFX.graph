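# Niagara system adding icy smoke and particles to the IcicleCane pickaxe character intro.
# IcySmoke fills a cylinder with smoke tinted by time of day and brief, delayed secondary sprites, with System.VelocityFade accelerating particle aging.
# Particle uses delayed, movement-based spawning with inherited velocity to shed sprite and mesh bits that drag, fall, flutter and swirl.
# Both emitters use spawn-group switches to vary behavior per particle type within one emitter.
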NiagaraSystem /BRCosmetics/Animation/Game/MainPlayer/CharIntros/s40/IcicleCane/FX/NS_CharIntros_Pickaxe_IcicleCane_SmokeFX
dependencies {
 usedDataInterfaces [
  &NiagaraDataInterfaceCurve
  &NiagaraDataInterfaceColorCurve
  &NiagaraDataInterfaceSpriteRendererInfo
 ]
 usedDynamicInputs [&VectorLength]
 usedModules [
  &SystemState
  &self:SystemSpawnScript.SystemScriptSource.SystemScriptGraph.NiagaraNodeAssignment_2.SetVariables_#1
 ]
}
emitters {
 IcySmoke {
  emitterUpdateScript [
   &EmitterState(
    Inactive Response=::NewEnumerator1
    Life Cycle Mode=::NewEnumerator1
    Loop Behavior=::NewEnumerator1
    Loop Duration Mode=::NewEnumerator1
    UseLoopDelay=false
   )
   &SpawnRate(SpawnRate=24, Use Spawn Probability=false) {module=SpawnRate001}
   &SpawnRate(
    Spawn Probability=&FloatFromCurve(
     FloatCurve=DataInterface() {propertyValues {Curve {keys [<at 0 0 RCIM_Constant>,<at 2 1 RCIM_Constant>]}}}
     CurveIndex=$Engine.System.Age
     Scale Curve=0.8
    )
    SpawnGroup=1
    SpawnRate=1.35
    Use Spawn Probability=true
   )
  ]
  particleSpawnScript [
   &InitializeParticle(
    Color=&MergeHelper_ColorSwitch(
     0=&RandomRangeLinearColor(Minimum=<rgba 0.2563 0.299 0.4271 0.6>, Maximum=<rgba 0.6 0.6667 1 0.8>)
     1=&RandomRangeLinearColor(Minimum=<rgba 0.6 0.9 1 1>, Maximum=<rgba 0.6 0.7 1 1>)
     2=&RandomRangeLinearColor(Minimum=<rgba 0.6 0.9067 1 1>, Maximum=<rgba 0.6 0.8067 1 1>)
     3=?
     Index=$Emitter.SpawnGroup
     Num=::NewEnumerator0
    )
    Lifetime=&MergeHelper_FloatSwitch(
     0=&RandomRangeFloat(Minimum=1.4, Maximum=2)
     1=&RandomRangeFloat(Minimum=0.18, Maximum=0.225)
     2=?
     3=?
     Index=$Emitter.SpawnGroup
     Num=::NewEnumerator0
    )
    Sprite Rotation Mode=::NewEnumerator2
    Sprite Size Mode=::NewEnumerator3
    Sprite UV Mode=::NewEnumerator3
    Uniform Sprite Size=&MergeHelper_FloatSwitch(
     0=&RandomRangeFloat(Minimum=17, Maximum=23.8)
     1=&RandomRangeFloat(Minimum=3.5, Maximum=4.35)
     2=&RandomRangeFloat(Minimum=25, Maximum=25)
     3=?
     Index=$Emitter.SpawnGroup
     Num=::NewEnumerator0
    )
    UsePositionOffset=false
    Write Material Random=true
   )
   &PartitionParticles(Number Of Partitions=&MergeHelper_IntSwitch(0=2, 1=1, 2=?, 3=?, 4=?, 5=?, Index=$Emitter.SpawnGroup, Num=::NewEnumerator0))
   &ShapeLocation(
    Cylinder Height=&MergeHelper_FloatSwitch(0=70, 1=65, 2=?, 3=?, Index=$Emitter.SpawnGroup, Num=::NewEnumerator0)
    Cylinder Height Midpoint=0
    Cylinder Radius=&MergeHelper_FloatSwitch(0=1, 1=1, 2=?, 3=?, Index=$Emitter.SpawnGroup, Num=::NewEnumerator0)
    MaskSpawnGroup=false
    Offset=<0 0 25>
    Offset Coordinate Space=::Local
    Offset Mode=::NewEnumerator0
    Override Seed=false
    Rotation Angle Type:ENiagara_AngleInput=::NewEnumerator0
    Rotation Axis=<0 0 1>
    Rotation Mode=::NewEnumerator1
    Shape Primitive=::NewEnumerator1
    Surface Only=false
    UseOwnerScale=true
   )
   &CameraOffset(Camera Offset Amount=&MergeHelper_FloatSwitch(0=10, 1=30, 2=?, 3=?, Index=$Emitter.SpawnGroup, Num=::NewEnumerator0))
   &AddVelocity(Velocity=&MergeHelper_VectorSwitch(
     0=&MergeHelper_VectorSwitch(
      0=&RandomRangeVector(Minimum=<0 0 0>, Maximum=<0 0 9>)
      1=<0 0 0>
      2=?
      3=?
      Index=$Particles.PartitionParticles.Partition
      Num=::NewEnumerator0
     )
     1=<0 0 0>
     2=?
     3=?
     Index=$Emitter.SpawnGroup
     Num=::NewEnumerator0
    ))
   /SetVariables(Particles.VisibilityTag=$Emitter.SpawnGroup) {isSetParameters=true}
   /SetVariables(Particles.DynamicParametersAlt:Vector4f=<1 1 1 1>) {isSetParameters=true}
   /SetVariables() {isSetParameters=true}
  ]
  particleUpdateScript [
   &ParticleState()
   /SetVariables(Particles.AgeSubUV:NiagaraFloat=&Add_Float(A=$Particles.AgeSubUV, B=$Engine.DeltaTime)) {isSetParameters=true}
   /SetVariables(Particles.Age:NiagaraFloat=&Add_Float(
     A=$Particles.Age
     B=&Multiply_Float(A=$System.VelocityFade, B=&Multiply_Float(A=$Particles.Lifetime, B=$Engine.DeltaTime))
    )) {isSetParameters=true}
   &GravityForce(Gravity=&MergeHelper_VectorSwitch(
     0=&MergeHelper_VectorSwitch(
      0=&RandomRangeVector(Minimum=<0 0 -400>, Maximum=<0 0 -635>)
      1=<0 0 0>
      2=?
      3=?
      Index=$Particles.PartitionParticles.Partition
      Num=::NewEnumerator0
     )
     1=<0 0 0>
     2=?
     3=?
     Index=$Particles.VisibilityTag
     Num=::NewEnumerator0
    ))
   &Drag(Use Linear Drag=true, Use Rotational Drag=true)
   &ScaleVelocity(Velocity Scale=&MakeVector(
     X=&FloatFromCurve(FloatCurve=DataInterface() {propertyValues {Curve {keys [{arriveTangent=0.0712, time=0, value=0}, {arriveTangent=1, leaveTangent=1, time=0.3, value=0.1}, <at 0.8 1>]}}})
     Y=1
     Z=1
    ))
   &ScaleColor(
    Scale Alpha=&MergeHelper_FloatSwitch(
     0=&FloatFromCurve(
      FloatCurve=DataInterface() {propertyValues {Curve {keys [
          {arriveTangent=7.9556, interpMode=RCIM_Cubic, leaveTangent=7.9556, tangentMode=RCTM_User, time=0, value=0}
          {arriveTangentWeight=0.0667, interpMode=RCIM_Cubic, leaveTangentWeight=0.1, tangentMode=RCTM_User, time=0.2, value=1}
          {arriveTangentWeight=0.1601, interpMode=RCIM_Cubic, leaveTangentWeight=0.2668, tangentMode=RCTM_User, time=0.5, value=1}
          {arriveTangent=-3.5681, tangentMode=RCTM_User, time=1, value=0}
         ]}}}
      Scale Curve=0.1
     )
     1=1
     2=?
     3=?
     Index=$Particles.VisibilityTag
     Num=::NewEnumerator0
    )
    Scale RGB=&MergeHelper_VectorSwitch(
     0=&MakeVectorFromLinearColorRGB(Color=&ColorFromCurve(
       ColorCurve=DataInterface() {propertyValues {
         AlphaCurve {keys [<at 0 1 RCIM_Cubic>]}
         BlueCurve {keys [<at 0.0932 1 RCIM_Cubic>,<at 0.125 1>,<at 0.5 1>,<at 0.875 1 RCIM_Cubic>,<at 0.905 1>]}
         GreenCurve {keys [
           <at 0.0932 0.5833 RCIM_Cubic>
           <at 0.125 1>
           <at 0.5 1>
           {arriveTangent=-1.0288, interpMode=RCIM_Cubic, leaveTangent=-1.0288, time=0.875, value=1}
           <at 0.905 0.5833>
          ]}
         RedCurve {keys [
           <at 0.0932 0.5 RCIM_Cubic>
           <at 0.125 1>
           <at 0.5 1>
           {arriveTangent=-1.2346, interpMode=RCIM_Cubic, leaveTangent=-1.2346, time=0.875, value=1}
           <at 0.905 0.5>
          ]}
        }}
       CurveIndex=Linked() {linkedVariable {name=NPC.FortniteNPC.RemappedTimeOfDay_0-1, type=NiagaraFloat}}
      ))
     1=&MakeVectorFromLinearColorRGB(Color=<rgba 1.05 1.3875 1.5 3>)
     2=?
     3=?
     Index=$Particles.VisibilityTag
     Num=::NewEnumerator0
    )
    ScaleA=true
    ScaleRGB=true
   )
   &ScaleSpriteSize(Scale Factor=&MergeHelper_Vector2Switch(
     0=&Vector2DFromFloat(Value=&FloatFromCurve(FloatCurve=DataInterface() {propertyValues {Curve {keys [{arriveTangent=0.3685, time=0, value=0.8}, <at 0.3 0.9>, {arriveTangent=0.1633, leaveTangent=0.1633, time=1, value=1}]}}}))
     1=&Vector2DFromFloat(Value=&FloatFromCurve(FloatCurve=DataInterface() {propertyValues {Curve {keys [
           {arriveTangent=0.3685, interpMode=RCIM_Cubic, time=0, value=1}
           {arriveTangent=-1, interpMode=RCIM_Cubic, leaveTangent=-1, time=0.3, value=0.9}
           {arriveTangent=0.1633, leaveTangent=0.1633, time=1, value=0}
          ]}}}))
     2=?
     3=?
     4=?
     Index=$Particles.VisibilityTag
     Num=::NewEnumerator0
    ))
   &SpriteRotationRate()
   &SubUVAnimation(
    Output Alternate Renderer Binding=true
    "Sprite Renderer "=DataInterface() {propertyValues {SpriteRenderer=&self:IcySmoke_0.NiagaraSpriteRendererProperties_1}}
    SubUV Lookup Index:NiagaraFloat=&Divide_Float(A=$Particles.AgeSubUV, B=$Particles.Lifetime)
    UseEndFrame=false
    UseStartFrame=false
   )
   &DynamicMaterialParameters(
    Index 0 Param 1=1
    Index 0 Param 4=1
    Module.Index0Float0Write=true
    Module.Index0Float1Write=true
    Module.Index0Float2Write=true
    Module.Index0Float3Write=true
   )
   &SolveForcesAndVelocity(Clamp Velocity=false, Limit Acceleration=false)
  ]
  rendererClasses [&NiagaraSpriteRendererProperties]
  renderers [
   {rendererClass=&NiagaraSpriteRendererProperties}
   {rendererClass=&NiagaraSpriteRendererProperties, rendererIndex=1}
  ]
 }
 Particle {
  emitterUpdateScript [
   &EmitterState(Life Cycle Mode=::NewEnumerator1, Loop Behavior=::NewEnumerator1, Loop Delay=1.15, Loop Duration=0.85, UseLoopDelay=true)
   &SpawnPerUnit(
    Max Movement Threshold=2500
    Module.Use Max Movement Threshold=false
    Module.Use Movement Tolerance=false
    Spawn Probability=0.85
    Spawn Spacing=10
    Use Spawn Probability=true
   )
   &SpawnPerUnit(
    Max Movement Threshold=2500
    Module.Use Max Movement Threshold=false
    Module.Use Movement Tolerance=false
    Spawn Group=1
    Spawn Probability=0.85
    Spawn Spacing=17
    Use Spawn Probability=true
   ) {module=SpawnPerUnit001}
   &SpawnPerUnit(
    Max Movement Threshold=2500
    Module.Use Max Movement Threshold=false
    Module.Use Movement Tolerance=false
    Spawn Group=2
    Spawn Spacing=10
    Use Spawn Probability=false
   ) {module=SpawnPerUnit002}
   /SetVariables(
    Emitter.PetalInnerColor:LinearColor=<rgba 0.4 0.55 1 1>
    Emitter.PetalMiddleColor:LinearColor=<rgba 0.65 0.9125 1 1>
    Emitter.PetalOuterColor:LinearColor=<rgba 0.3583 0.3 1 1>
   ) {isSetParameters=true}
  ]
  particleSpawnScript [
   &InitializeParticle(
    Color=<rgba 0.7 0.925 1 1>
    Lifetime=&MergeHelper_FloatSwitch(
     0=&RandomRangeFloat(Minimum=0.7, Maximum=1)
     1=&RandomRangeFloat(Minimum=0.4, Maximum=0.65)
     2=&RandomRangeFloat(Minimum=0.45, Maximum=0.7)
     3=?
     Index=$Emitter.SpawnGroup
     Num=::NewEnumerator1
    )
    Mesh Scale Mode=::NewEnumerator3
    Mesh Uniform Scale=&RandomRangeFloat(Minimum=0.07, Maximum=0.126)
    Sprite Rotation Mode=::NewEnumerator2
    Sprite Size=&Multiply_Vector2D_ByFloat(Float=&RandomRangeFloat(Minimum=0.1715, Maximum=0.4865), Vector2D=<18.5 7>)
    Sprite Size Mode=::NewEnumerator4
    Sprite UV Mode=::NewEnumerator3
    UsePositionOffset=false
    Write Material Random=true
   )
   &ShapeLocation(
    MaskSpawnGroup=false
    Offset=<0 0 85>
    Offset Coordinate Space=::Local
    Offset Mode=::NewEnumerator0
    Override Seed=false
    Sphere Radius=20
    UseOwnerScale=true
   )
   &InheritVelocity(
    Inherited Velocity Speed Limit:NiagaraFloat=&MergeHelper_FloatSwitch(
     0=&RandomRangeFloat(Minimum=200, Maximum=350)
     1=&RandomRangeFloat(Minimum=350, Maximum=600)
     2=&RandomRangeFloat(Minimum=80, Maximum=140)
     3=?
     Index=$Emitter.SpawnGroup
     Num=::NewEnumerator1
    )
    Speed Limit=true
    UseSpeedThreshold=false
   )
   &CameraOffset(Camera Offset Amount=&MergeHelper_FloatSwitch(0=7, 1=3, 2=7, 3=?, Index=$Emitter.SpawnGroup, Num=::NewEnumerator1))
   /SetVariables(Particles.VisibilityTag=$Emitter.SpawnGroup) {isSetParameters=true}
  ]
  particleUpdateScript [
   &ParticleState()
   &ScaleMeshSize(Scale Factor=&VectorFromFloat(Value=&FloatFromCurve(FloatCurve=DataInterface() {propertyValues {Curve {keys [<at 0 1 RCIM_Cubic>,<at 1 0 RCIM_Cubic>]}}}, Scale Curve=2)))
   &Drag(
    Drag=&Multiply_Float(
     A=&UniformRangedFloat(Minimum=1.2, Maximum=1.5)
     B=&FloatFromCurve(
      FloatCurve=DataInterface() {propertyValues {Curve {keys [
          {arriveTangent=-5.3551, interpMode=RCIM_Cubic, leaveTangent=-5.3551, tangentMode=RCTM_User, time=0, value=0}
          {arriveTangent=-138.8873, interpMode=RCIM_Cubic, leaveTangent=-138.8873, tangentMode=RCTM_User, time=0.05, value=7}
          <at 0.2 0.5>
         ]}}}
      Scale Curve=4
     )
    )
    Use Linear Drag=true
    Use Rotational Drag=false
   )
   &SubUVAnimation(
    Output Alternate Renderer Binding=true
    "Sprite Renderer "=DataInterface() {propertyValues {SpriteRenderer=&self:Particle_0.NiagaraSpriteRendererProperties_1}}
    UseEndFrame=false
    UseStartFrame=false
   )
   &ScaleColor(
    Scale RGB=&VectorFromFloat(Value=&FloatFromCurve(FloatCurve=DataInterface() {propertyValues {Curve {keys [{arriveTangent=-2.7905, interpMode=RCIM_Cubic, leaveTangent=-2.7905, tangentMode=RCTM_User, time=0, value=2}, <at 1 1 RCIM_Cubic>]}}}))
    ScaleA=true
    ScaleRGB=true
   )
   &GravityForce(Gravity=&RandomRangeVector(Minimum=<0 0 -150>, Maximum=<0 0 -200>))
   &CurlNoiseForce(
    Bias Noise=false
    Noise Frequency=20
    Noise Strength=&MergeHelper_FloatSwitch(
     0=&RandomRangeFloat(Minimum=225, Maximum=350)
     1=0
     2=0
     3=?
     Index=$Particles.VisibilityTag
     Num=::NewEnumerator1
    )
    Pan Noise=false
   )
   &ScaleSpriteSize(Scale Factor=&Lerp_Vector2(
     A=&Multiply_Vector2(
      A=&MakeVector2D(
       X=&Multiply_Float(
        A=1
        B=&Sine(
         Normalized Angle=$Emitter.Age
         Period=&Multiply_Float(A=0.7, B=&Add_Float(A=$Particles.Initial.MaterialRandom, B=0.65))
         Scale=1
         Bias=0
        )
       )
       Y=1
      )
      B=&Vector2DFromFloat(Value=&FloatFromCurve(FloatCurve=DataInterface() {propertyValues {Curve {keys [<at 0 1.1>,<at 0.75 1.1>,<at 1 0>]}}}))
     )
     B=<1 1>
     Alpha=&FloatFromCurve(FloatCurve=DataInterface() {propertyValues {Curve {keys [<at 0.0574 1>,<at 0.1545 0>]}}})
    ))
   &SpriteRotationRate(Rotation Rate=&Multiply_Float(
     A=&UniformAOrBFloat(A=-1, Evaluation Type=::NewEnumerator0, Recalculate Random Each Loop=true, Override Seed=false)
     B=&MergeHelper_FloatSwitch(
      0=&UniformRangedFloat(Minimum=50, Maximum=200)
      1=0
      2=0
      3=?
      Index=$Particles.VisibilityTag
      Num=::NewEnumerator1
     )
    ))
   &FN_Sprite_Attributes(
    Color=&MakeLinearColorFromVectorAndFloat(
     "Vector (RGB)"=&MakeVectorFromLinearColorRGB(Color=&RandomRangeLinearColor(Minimum=<rgba 0.2563 0.299 0.4271 0.6>, Maximum=<rgba 0.6 0.6667 1 0.8>))
     "Float (Alpha)"=&FloatFromCurve(
      FloatCurve=DataInterface() {propertyValues {Curve {keys [
          {arriveTangent=7.9556, interpMode=RCIM_Cubic, leaveTangent=7.9556, tangentMode=RCTM_User, time=0, value=0}
          {arriveTangentWeight=0.0667, interpMode=RCIM_Cubic, leaveTangentWeight=0.1, tangentMode=RCTM_User, time=0.2, value=1}
          {arriveTangentWeight=0.1601, interpMode=RCIM_Cubic, leaveTangentWeight=0.2668, tangentMode=RCTM_User, time=0.5, value=1}
          {arriveTangent=-3.5681, tangentMode=RCTM_User, time=1, value=0}
         ]}}}
      Scale Curve=0.08
     )
    )
    Create Color Attribute=true
    Create SpriteSize Attribute=true
    SpriteSize=&Multiply_Vector2D_ByFloat(
     Float=&MakeFloatFromVector2d(Vector 2D=&Vector2DFromFloat(Value=&FloatFromCurve(FloatCurve=DataInterface() {propertyValues {Curve {keys [
            {arriveTangent=0.4145, interpMode=RCIM_Cubic, leaveTangent=0.4145, tangentMode=RCTM_User, time=0, value=0.8}
            {arriveTangent=0.2495, interpMode=RCIM_Cubic, leaveTangent=0.2495, tangentMode=RCTM_User, time=0.3, value=0.9}
            {arriveTangentWeight=0.2333, interpMode=RCIM_Cubic, tangentMode=RCTM_User, time=1, value=1}
           ]}}})))
     Vector2D=&Vector2DFromFloat(Value=&RandomRangeFloat(Minimum=49, Maximum=63))
    )
    Write to Module Namespace=true
   )
   &SolveForcesAndVelocity(Clamp Velocity=false, Limit Acceleration=false)
  ]
  rendererClasses [&NiagaraSpriteRendererProperties, &NiagaraMeshRendererProperties]
  renderers [
   {rendererClass=&NiagaraSpriteRendererProperties}
   {rendererClass=&NiagaraMeshRendererProperties, rendererIndex=1}
   {rendererClass=&NiagaraSpriteRendererProperties, rendererIndex=2}
  ]
 }
}
summary {
 systemName=NS_CharIntros_Pickaxe_IcicleCane_SmokeFX
 emitters [
  {emitterName=IcySmoke, rendererClasses [&NiagaraSpriteRendererProperties]}
  {emitterName=Particle, rendererClasses [&NiagaraSpriteRendererProperties, &NiagaraMeshRendererProperties]}
 ]
 userVariables []
}
